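# Translucent, unlit, two-sided Niagara sprite material for multicolored confetti on the Vivid Male glider.
# Particle random, scaled, blends through five color parameters via chained subtract-saturate lerps; the result is squared, tinted by particle color and fed to MF_FX_EmissiveAdjustments.
# The same random value drives FlipBook UVs into the confetti texture, rounded into a hard-edged mask.
# A separate branch multiplies sparkle texture channels, one exponent-shaped, by particle alpha.
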Material /BRCosmetics/Gadgets/Assets/VinderTech_GliderChute/Glider_Vivid_Male/FX/Materials/M_Glider_Vivid_Male_Confetti
graph {
 properties {
  bUsedWithNiagaraSprites=true
  blendMode=BLEND_Translucent
  pixelDepthOffsetMode=PDOM_Legacy
  shadingModel=MSM_Unlit
  twoSided=true
  usedShadingModels=Unlit
 }
 nodes [
  n1 = Multiply(A=ParticleRandom()) {constB=4}
  n2 = LinearInterpolate(
   A=VectorParameter() {parameterName=Color 1, defaultValue=<rgba 0.4969 0.227 0.624 1>}.RGB
   B=VectorParameter() {parameterName=Color 2, defaultValue=<rgba 1 0.011 0.2664 0.2>}.RGB
   Alpha=Saturate(None=@n1)
  )
  n3 = VectorParameter() {parameterName=Color 3, defaultValue=<rgba 1 0.8549 0 0.3>}
  n4 = Subtract(A=@n1)
  n5 = Saturate(None=@n4)
  n6 = Subtract(A=@n4)
  n7 = LinearInterpolate(
   A=LinearInterpolate(
    A=LinearInterpolate(A=@n2, B=@n3.RGB, Alpha=@n5)
    B=VectorParameter() {parameterName=Color 4, defaultValue=<rgba 0.011 1 0.2747 0.2>}.RGB
    Alpha=Saturate(None=@n6)
   )
   B=VectorParameter() {parameterName=Color 5, defaultValue=<rgba 0.011 0.4508 1 0.1>}.RGB
   Alpha=Saturate(None=Subtract(A=@n6))
  )
  n8 = ParticleColor()
  MaterialFunctionCall(Emissive=Multiply(A=Multiply(A=@n7, B=@n7), B=@n8.RGB)) {materialFunction=&MF_FX_EmissiveAdjustments}
  n9 = TextureSample() {texture=/Game/Effects/Fort_Effects/Textures/Sparkle/T_Sparkle_01}
  n10 = ScalarParameter() {defaultValue=1, parameterName=Exponent}
  Multiply(A=@n8.A, B=Multiply(A=@n9.R, B=Power(Base=@n9.G, Exp=@n10)))
  n11 = TextureSample(UVs=MaterialFunctionCall("Animation  Phase (0-1)"=@n1) {materialFunction=&FlipBook}.UVs) {
   samplerType=SAMPLERTYPE_LinearGrayscale
   texture=/BRCosmetics/Gadgets/Assets/VinderTech_GliderChute/Glider_Vivid_Male/FX/Textures/T_Glider_Vivid_Confetti
  }
  Round(None=@n11.R)
 ]
}
