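# Translucent, unlit, two-sided Niagara ribbon material for the trail of the Vivid Male glider.
# Five tunable colors blend in bands across the ribbon's width and are squared, with a brighter base color lerped in by distance from a local-space offset point.
# Opacity math combines length and edge fades, a time-animated sine wave eroding the tail, particle alpha and depth fade.
# Dynamic parameters drive wave speed, wave intensity and both fades.
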Material /BRCosmetics/Gadgets/Assets/VinderTech_GliderChute/Glider_Vivid_Male/FX/Materials/M_Glider_Vivid_Male_Trail
graph {
 properties {
  bUsedWithNiagaraRibbons=true
  blendMode=BLEND_Translucent
  pixelDepthOffsetMode=PDOM_Legacy
  shadingModel=MSM_Unlit
  twoSided=true
  usedShadingModels=Unlit
 }
 nodes [
  Constant()
  n1 = TextureCoordinate()
  n2 = ComponentMask(None=@n1) {r=true}
  n3 = OneMinus(None=@n2)
  n4 = ComponentMask(None=@n1) {g=true}
  n5 = DynamicParameter() {defaultValue=<rgba 0.7 0.1 0.1 0.6>, paramNames [TrailWaveSpeed, TrailWaveIntensity, EdgeFade, LengthFade]}
  n6 = Multiply(A=Add(A=@n4, B=Time()), B=@n5.TrailWaveSpeed)
  n7 = Multiply(A=ConstantBiasScale(None=Sine(None=@n6)), B=@n5.TrailWaveIntensity)
  n8 = Power(Base=Multiply(A=@n3, B=Subtract(A=@n3, B=@n7))) {constExponent=1.5}
  n9 = DynamicParameter() {defaultValue=<rgba 0.7 0.1 0.1 0.6>, paramNames [TrailWaveSpeed, TrailWaveIntensity, EdgeFade, LengthFade]}
  n10 = OneMinus(None=Abs(None=ConstantBiasScale(None=@n4) {bias=-0.5, scale=2}))
  n11 = Multiply(A=Saturate(None=Divide(A=@n8, B=@n9.LengthFade)), B=Saturate(None=Divide(A=@n10, B=@n9.EdgeFade)))
  DepthFade(Opacity=Multiply(A=@n11, B=ParticleColor().A), FadeDistance=ScalarParameter() {defaultValue=44, parameterName=Depth Fade Distance})
  n12 = Multiply(A=OneMinus(None=@n4)) {constB=4}
  n13 = LinearInterpolate(
   A=VectorParameter() {parameterName=Color 1, defaultValue=<rgba 0.4969 0.227 0.624 1>}.RGB
   B=VectorParameter() {parameterName=Color 2, defaultValue=<rgba 1 0.011 0.2664 0.2>}.RGB
   Alpha=Saturate(None=@n12)
  )
  n14 = VectorParameter() {parameterName=Color 3, defaultValue=<rgba 1 0.8549 0 0.3>}
  n15 = Subtract(A=@n12)
  n16 = Saturate(None=@n15)
  n17 = Subtract(A=@n15)
  n18 = LinearInterpolate(
   A=LinearInterpolate(
    A=LinearInterpolate(A=@n13, B=@n14.RGB, Alpha=@n16)
    B=VectorParameter() {parameterName=Color 4, defaultValue=<rgba 0.011 1 0.2747 0.2>}.RGB
    Alpha=Saturate(None=@n17)
   )
   B=VectorParameter() {parameterName=Color 5, defaultValue=<rgba 0.011 0.4508 1 0.1>}.RGB
   Alpha=Saturate(None=Subtract(A=@n17))
  )
  n19 = Multiply(A=@n18, B=@n18)
  n20 = VectorParameter() {parameterName=Base Color, defaultValue=<rgba 33 18.755 9.9 9>}
  n21 = Distance(A=TransformPosition(None=Constant3Vector() {constant=<rgba 0 100 0 0>}) {transformType=TRANSFORMPOSSOURCE_World}, B=WorldPosition().XYZ)
  n22 = ScalarParameter() {defaultValue=22, parameterName=Core Size}
  n23 = Saturate(None=Divide(A=Subtract(A=@n21, B=@n22), B=ScalarParameter() {defaultValue=33, parameterName=Falloff Size}))
  LinearInterpolate(A=@n19, B=@n20.RGB, Alpha=Power(Base=OneMinus(None=@n23), Exp=@n20.A))
  n24 = ScalarParameter() {defaultValue=0.1, parameterName=FresnelExp}
  MaterialFunctionCall(Invert Fresnel=StaticBool() {value=true}, Power=@n24) {materialFunction=&Fresnel_Function}
  MaterialFunctionCall(Emissive=@n19) {materialFunction=&MF_FX_EmissiveAdjustments}
  OneMinus(None=Fresnel(ExponentIn=@n24))
  ScalarParameter() {defaultValue=0.1, parameterName=Edge Fade}
  ScalarParameter() {defaultValue=0.6, parameterName=Length Fade}
  ScalarParameter() {defaultValue=0.7, parameterName=Tail Wave Speed}
  ScalarParameter() {defaultValue=0.1, parameterName=Tail Wave Intensity}
  n25 = Divide(A=@n2, B=ScalarParameter() {defaultValue=0.3, parameterName=Base Fade})
  VertexInterpolator(VS=Multiply(A=Saturate(None=@n25), B=@n11))
 ]
}
