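# Niagara trail effect for the Vivid Male glider cosmetic, layering rainbow socket glows, confetti, mesh confetti, cone streak contrails, ribbons and a light.
# Emitters sample User.SkeletalMesh for attachment, while ForwardDot scales most spawn rates and RightDot skews many velocities sideways.
# FrontEnd modifiers switch rates, velocities and sizes between gameplay and menu values.
# Splash spawns two multicolour ribbons whose velocity scales with owner speed and alpha with time of day.
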NiagaraSystem /BRCosmetics/Gadgets/Assets/VinderTech_GliderChute/Glider_Vivid_Male/FX/NS_Glider_Vivid_Male_Trails
dependencies {
 usedDataInterfaces [&NiagaraDataInterfaceCurve, &NiagaraDataInterfaceArrayFloat3, &NiagaraDataInterfaceColorCurve]
 usedDynamicInputs [
  /Niagara/DynamicInputs/Helpers/SimulationPosition
  &Lerp_Vector
  /Game/Effects/Niagara/DynamicInputs/Random/FN_SeededRandomRangeFloat
 ]
 usedModules [&SystemState, &FN_Integrator]
}
emitters {
 AttachedGlows {
  emitterUpdateScript [
   &EmitterState(Inactive Response=::NewEnumerator1, Life Cycle Mode=::NewEnumerator1, Loop Duration=0, UseLoopDelay=false)
   &SpawnRate(
    SpawnRate=&FN_FrontEndFloatModifier(
     Gameplay=&Lerp_Float(A=15, B=25, Alpha=&ClampFloat(Float=$User.ForwardDot))
     Front End=25
    )
    Use Spawn Probability=false
   )
  ]
  particleSpawnScript [
   &InitializeParticle(
    Lifetime Max=0.25
    Lifetime Min=0.15
    Lifetime Mode=::NewEnumerator1
    Sprite Size Mode=::NewEnumerator1
    Uniform Sprite Size Max=80
    Uniform Sprite Size Min=50
    UsePositionOffset=false
    Write Material Random=true
   ) {module=InitializeParticle001}
   &CameraOffset(Camera Offset Amount=5)
   &BoxLocation(Box Offset=<-20 0 0>, Box Size=<40 1 1>, MaskSpawnGroup=false, Module.Override Seed=false, Rotate Box=false, Surface Only=false)
   /SetVariables(
    Particles.TestVec:Vector3f=&Add_Vector(A=Linked() {linkedVariable {name=Particles.Position, type=NiagaraPosition}}, B=<20 0 0>)
    Particles.VectorPos:NiagaraFloat=&VectorLength(Vector=&Multiply_Vector(A=Linked() {linkedVariable {name=Particles.Position, type=NiagaraPosition}}, B=<1 0 0>))
   ) {isSetParameters=true}
  ]
  particleUpdateScript [
   &SkeletalMeshLocation(
    MaskSpawnGroup=false
    Mesh Sampling Type=::NewEnumerator5
    Sampled Position Offset=&RandomRangeVector(Minimum=<0 20 4>, Maximum=<40 20 5>)
    Skeletal Mesh=$User.SkeletalMesh
    Socket Sampling Mode=::NewEnumerator5
   )
   &ScaleColor(
    Scale RGB=&MakeVectorFromLinearColorRGB(Color=&ColorFromCurve(
      ColorCurve=DataInterface() {propertyValues {
        AlphaCurve {keys [<at 0 1 RCIM_Cubic>,<at 1 1 RCIM_Cubic>]}
        BlueCurve {keys [<at 0 1 RCIM_Cubic>,<at 0.2726 0.0268>,<at 0.5066 0>,<at 0.7304 0>,<at 1 0.455 RCIM_Cubic>]}
        GreenCurve {keys [<at 0 0.3907 RCIM_Cubic>,<at 0.2726 1>,<at 0.5066 0.7877>,<at 0.7304 0.2579>,<at 1 0 RCIM_Cubic>]}
        RedCurve {keys [<at 0 0 RCIM_Cubic>,<at 0.2726 0>,<at 0.5066 1>,<at 0.7304 1>,<at 1 1 RCIM_Cubic>]}
       }}
      CurveIndex=&NormalizeFloat(Value=$Particles.VectorPos, Min=0, Max=40)
     ))
    ScaleA=true
    ScaleRGB=true
   )
   &ParticleState()
   &SolveForcesAndVelocity(Clamp Velocity=false, Limit Acceleration=false)
   &ScaleSpriteSize(
    Scale Sprite Size Mode=::NewEnumerator0
    Uniform Scale Factor=&FN_FrontEndFloatModifier(Gameplay=1, Front End=0.7)
   )
  ]
  rendererClasses [&NiagaraSpriteRendererProperties]
  renderers [{rendererClass=&NiagaraSpriteRendererProperties}]
 }
 Confetti {
  emitterUpdateScript [
   &EmitterState()
   &SpawnRate(
    SpawnRate=&FN_FrontEndFloatModifier(
     Gameplay=&Lerp_Float(A=150, B=500, Alpha=&ClampFloat(Float=$User.ForwardDot))
     Front End=200
    )
    Use Spawn Probability=false
   )
  ]
  particleSpawnScript [
   &InitializeParticle(
    Lifetime=&FN_FrontEndFloatModifier(Gameplay=&RandomRangeFloat(Minimum=0.1, Maximum=0.3), Front End=0.52)
    Position Mode=::NewEnumerator0
    Sprite Rotation Mode=::NewEnumerator2
    Sprite Size Mode=::NewEnumerator1
    Sprite UV Mode=::NewEnumerator3
    Uniform Sprite Size Max=11
    Write Material Random=true
   )
   &AddVelocity(Velocity=&FN_FrontEndVector3Modifier(
     Gameplay=&MakeVector(
      X=&Lerp_Float(A=600, B=-600, Alpha=$Particles.MaterialRandom)
      Y=&RandomRangeFloat(Minimum=100, Maximum=500)
      Z=&RandomRangeFloat(Minimum=200, Maximum=350)
     )
     Front End=&RandomRangeVector(Minimum=<300 200 50>, Maximum=<-300 1000 150>)
    )) {module=AddVelocity001}
   &AddVelocity(Velocity=&Multiply_VectorByFloat(Vector=<400 0 0>, Float=&Multiply_Float(A=$User.RightDot, B=1)))
   &SystemLocation(
    MaskSpawnGroup=false
    Offset=&MakeVector(
     X=&Lerp_Float(A=15, B=-15, Alpha=$Particles.MaterialRandom)
     Y=95
     Z=&Lerp_Float(A=-5, B=-5, Alpha=&RandomRangeFloat(Minimum=0, Maximum=1))
    )
   ) {enabled=false}
   &SkeletalMeshLocation(
    Bone Sampling Mode=::NewEnumerator4
    MaskSpawnGroup=false
    Sampled Position Offset=&RandomRangeVector(Minimum=<13 95 5>, Maximum=<-13 95 -5>)
    Skeletal Mesh=$User.SkeletalMesh
   )
  ]
  particleUpdateScript [
   &InheritVelocity(Inherited Velocity Amount Scale=<0 0 1>, Speed Limit=true, UseSpeedThreshold=false)
   &AccelerationForce(Acceleration=&FN_FrontEndVector3Modifier(Gameplay=<0 0 0.1>, Front End=<0 0 0>))
   &Drag(Drag=5, Use Linear Drag=true, Use Rotational Drag=false)
   &SolveForcesAndVelocity(Clamp Velocity=false, Limit Acceleration=false)
   &ParticleState()
   &SpriteRotationRate(Rotation Rate=&RandomRangeFloat(Minimum=-666, Maximum=666))
   &ScaleSpriteSize(Uniform Curve Sprite Scale=DataInterface() {propertyValues {Curve {keys [{tangentMode=RCTM_User, time=0, value=0}, <at 0.2 1>, {tangentMode=RCTM_User, time=1, value=0}]}}})
   &ScaleColor(
    Scale Alpha=&FloatFromCurve(FloatCurve=DataInterface() {propertyValues {Curve {keys [{interpMode=RCIM_Cubic, tangentMode=RCTM_User, time=0, value=1}, {interpMode=RCIM_Cubic, tangentMode=RCTM_User, time=1, value=0}]}}})
    Scale RGB=&VectorFromFloat(Value=&FloatFromCurve(FloatCurve=DataInterface() {propertyValues {Curve {keys [
          {arriveTangent=-28.3055, interpMode=RCIM_Cubic, leaveTangent=-28.3055, tangentMode=RCTM_User, time=0.1, value=10}
          {interpMode=RCIM_Cubic, tangentMode=RCTM_User, time=1, value=1}
         ]}}}))
    ScaleA=true
    ScaleRGB=true
   )
  ]
  rendererClasses [&NiagaraSpriteRendererProperties]
  renderers [{rendererClass=&NiagaraSpriteRendererProperties}]
 }
 FN_Contrails_ConeStreak {
  emitterUpdateScript [
   &EmitterState(Life Cycle Mode=::NewEnumerator1, Loop Duration=0, UseLoopDelay=false)
   &SpawnRate(
    SpawnRate=&FN_FrontEndFloatModifier(
     Gameplay=&Lerp_Float(A=10, B=30, Alpha=&ClampFloat(Float=$User.ForwardDot))
     Front End=30
    )
    Use Spawn Probability=true
   )
  ]
  particleSpawnScript [
   &FN_SetSpawnGroup(Number of Spawn Groups=2) {module=FN_SetSpawnGroup001}
   /Niagara/Modules/Spawn/Initialization/InitializeParticle(
    Color=&UniformRangedLinearColor(Minimum=<rgba 0 159.0469 500 1>, Maximum=<rgba 179.0069 69.217 500 1>)
    Lifetime=0.5
    Mesh Scale=&UniformRangedVector(Minimum=<0.2 0.2 0.2>, Maximum=<1 1 2>)
    Module.Write Color=true
    Module.Write Lifetime=true
    Module.Write Mass=true
    Module.Write Material Random=true
    Module.Write Position=false
    Module.Write Scale=true
    Module.Write SpriteRotation=false
    Module.Write SpriteSize=false
    Position=/Niagara/Functions/Localspace/SimulationPosition()
   )
   &AddVelocity(Velocity=&FN_FrontEndVector3Modifier(
     Gameplay=&Multiply_Vector(
      A=<0 400 100>
      B=&MakeVector(X=1, Y=&Lerp_Float(A=0.5, B=1, Alpha=&ClampFloat(Float=$User.ForwardDot)), Z=1)
     )
     Front End=&RandomRangeVector(Minimum=<0 500 20>, Maximum=<0 600 70>)
    )) {module=AddVelocity001}
   &AddVelocity(Velocity=&Multiply_VectorByFloat(Vector=<200 0 0>, Float=$User.RightDot))
  ]
  particleUpdateScript [
   &ParticleState()
   &SkeletalMeshLocation(
    Bone / Socket Index=$Particles.SpawnGroup
    MaskSpawnGroup=false
    Mesh Sampling Type=::NewEnumerator5
    Skeletal Mesh=$User.SkeletalMesh
    Socket Sampling Mode=::NewEnumerator5
   )
   &SolveForcesAndVelocity(Clamp Velocity=false, Limit Acceleration=false)
   &ScaleMeshSize(Scale Factor=&MakeVector(
     X=&Lerp_Float(A=1, B=40, Alpha=$Particles.NormalizedAge)
     Y=&Lerp_Float(A=1, B=40, Alpha=$Particles.NormalizedAge)
     Z=&Lerp_Float(A=1, B=250, Alpha=$Particles.NormalizedAge)
    ))
   &DynamicMaterialParameters(
    Index 0 Param 1=$Particles.NormalizedAge
    Index 0 Param 2=HlslExpression() {hlslExpression=Particles.MaterialRandom*0.5 + 0.5}
    Module.Param0WriteEnabled=true
    Module.Param1WriteEnabled=true
    Module.Param2WriteEnabled=true
    Module.Param3WriteEnabled=true
   )
   /Niagara/Modules/Math/Flight_Orientation("Foward Vector (Mesh Space)":Vector3f=<0 1 0>, "Right Vector (Mesh Space)":Vector3f=<0 0 1>)
   &ScaleColor(
    Scale RGB=&VectorFromFloat(Value=&FloatFromCurve(FloatCurve=DataInterface() {propertyValues {Curve {keys [
          {arriveTangent=-28.3055, interpMode=RCIM_Cubic, leaveTangent=-28.3055, tangentMode=RCTM_User, time=0.1, value=10}
          {interpMode=RCIM_Cubic, tangentMode=RCTM_User, time=1, value=1}
         ]}}}))
    ScaleA=false
    ScaleRGB=true
   )
  ]
  rendererClasses [&NiagaraMeshRendererProperties]
  renderers [{rendererClass=&NiagaraMeshRendererProperties, rendererIndex=0}]
 }
 Light {
  emitterUpdateScript [
   &EmitterState()
   &SpawnBurst_Instantaneous(
    Spawn Probability=&FN_FrontEndFloatModifier(
     Gameplay=&Lerp_Float(A=0.75, B=1, Alpha=&ClampFloat(Float=$User.ForwardDot))
     Front End=1
    )
    Use Spawn Probability=true
    UseLoopCountLimit=false
   )
  ]
  particleSpawnScript [
   /Niagara/Modules/Spawn/Initialization/V2/InitializeParticle(Color Mode=::NewEnumerator0, UsePositionOffset=false, Write Material Random=true)
   &SkeletalMeshLocation(
    Interpolate Mesh Sampling=false
    MaskSpawnGroup=false
    Mesh Sampling Type=::NewEnumerator5
    Sampled Position Offset=<20 20 0>
    Skeletal Mesh=$User.SkeletalMesh
    Socket Sampling Mode=::NewEnumerator5
   )
  ]
  particleUpdateScript [
   &ParticleState()
   &FN_Light_Attributes(
    Color=&ColorFromCurve(ColorCurve=DataInterface() {propertyValues {
       AlphaCurve {keys [<at 0 1 RCIM_Cubic>,<at 1 1 RCIM_Cubic>]}
       BlueCurve {keys [<at 0 1 RCIM_Cubic>,<at 0.2726 0.0268>,<at 0.5066 0>,<at 0.7304 0>,<at 1 0.455 RCIM_Cubic>]}
       GreenCurve {keys [<at 0 0.3907 RCIM_Cubic>,<at 0.2726 1>,<at 0.5066 0.7877>,<at 0.7304 0.2579>,<at 1 0 RCIM_Cubic>]}
       RedCurve {keys [<at 0 0 RCIM_Cubic>,<at 0.2726 0>,<at 0.5066 1>,<at 0.7304 1>,<at 1 1 RCIM_Cubic>]}
      }})
    Create Color Attribute=true
    Create Exponent Attribute=true
    Create Position Attribute=true
    Create Radius Attribute=true
    Position:NiagaraPosition=$Particles.Position
    Radius=300
    Scale Color=10
   )
  ]
  rendererClasses [&NiagaraLightRendererProperties]
  renderers [{rendererClass=&NiagaraLightRendererProperties, rendererIndex=0}]
 }
 Meshes {
  emitterUpdateScript [
   &EmitterState()
   &SpawnRate(
    SpawnRate=&FN_FrontEndFloatModifier(
     Gameplay=&Lerp_Float(A=10, B=30, Alpha=&ClampFloat(Float=$User.ForwardDot))
     Front End=30
    )
    Use Spawn Probability=true
   )
  ]
  particleSpawnScript [
   &InitializeParticle(
    Lifetime=&FN_FrontEndFloatModifier(Gameplay=0.5, Front End=0.7)
    Mesh Scale Mode=::NewEnumerator3
    UsePositionOffset=false
    Write Material Random=true
   )
   &AddVelocity(Velocity=&FN_FrontEndVector3Modifier(
     Gameplay=&RandomRangeVector(Minimum=<400 100 200>, Maximum=<-400 600 250>)
     Front End=&RandomRangeVector(Minimum=<200 300 100>, Maximum=<-200 1200 150>)
    ))
   &AddVelocity(Velocity=&Multiply_VectorByFloat(Vector=<400 0 0>, Float=&Multiply_Float(A=$User.RightDot, B=1))) {module=AddVelocity001}
   /SetVariables(
    Particles.MeshIndex=&RandomRangeInt(Minimum=0, Maximum=12)
    Particles.Scale=&VectorFromFloat(Value=&FloatFromCurve(
      FloatCurve=DataInterface() {propertyValues {Curve {keys [
          {interpMode=RCIM_Cubic, tangentMode=RCTM_User, time=0, value=0.7}
          {arriveTangent=0.7796, interpMode=RCIM_Cubic, leaveTangent=0.7796, tangentMode=RCTM_User, time=1, value=1}
         ]}}}
      CurveIndex=&RandomRangeFloat(Minimum=0, Maximum=1)
     ))
   ) {isSetParameters=true}
   &InitialMeshOrientation(
    Orientation Axis=<0 -1 0>
    Orientation Coordinate Space=::Local
    Rotation=&RandomRangeVector(Minimum=<0.1 0.1 0>, Maximum=<-0.1 -0.1 1>)
    Use Orientation Vector=true
    Use Rotation Vector=true
   )
   &CameraOffset(Camera Offset Amount=15)
   /SetVariables(Particles.RandRot:NiagaraFloat=&RandomRangeFloat(Minimum=-1, Maximum=1)) {isSetParameters=true}
   &SkeletalMeshLocation(
    Bone Sampling Mode=::NewEnumerator4
    MaskSpawnGroup=false
    Sampled Position Offset=&RandomRangeVector(Minimum=<13 95 5>, Maximum=<-13 95 -5>)
    Skeletal Mesh=$User.SkeletalMesh
   )
  ]
  particleUpdateScript [
   &InheritVelocity(Inherited Velocity Amount Scale=<0 0 1>, Speed Limit=true, UseSpeedThreshold=false)
   &AccelerationForce(Acceleration=&FN_FrontEndVector3Modifier(Gameplay=<0 0 0.3>, Front End=<0 0 0>))
   &Drag(Drag=5, Use Linear Drag=true, Use Rotational Drag=false)
   &SolveForcesAndVelocity(Clamp Velocity=false, Limit Acceleration=false, Manually Enable Rotational Solver=true)
   &ParticleState()
   &ScaleMeshSize(Scale Factor=&VectorFromFloat(Value=&FloatFromCurve(FloatCurve=DataInterface() {propertyValues {Curve {keys [
          {interpMode=RCIM_Cubic, leaveTangentWeight=0.9127, tangentMode=RCTM_User, time=0, value=0}
          {arriveTangentWeight=0.1054, interpMode=RCIM_Cubic, leaveTangentWeight=0.246, tangentMode=RCTM_User, time=0.2, value=0.9847}
          {arriveTangentWeight=0.5896, interpMode=RCIM_Cubic, tangentMode=RCTM_User, time=1, value=0}
         ]}}})))
   &SolveRotationalForcesAndVelocity() {enabled=false}
   &ScaleColor(
    Scale RGB=&VectorFromFloat(Value=&FloatFromCurve(FloatCurve=DataInterface() {propertyValues {Curve {keys [
          {interpMode=RCIM_Cubic, tangentMode=RCTM_User, time=0, value=0.3}
          {arriveTangent=1.6492, interpMode=RCIM_Cubic, leaveTangent=1.6492, tangentMode=RCTM_User, time=1, value=1}
         ]}}}))
    ScaleA=false
    ScaleRGB=true
   )
   &UpdateMeshOrientation(Rotation Rate=$Particles.RandRot, Rotation Vector=&RandomRangeVector(Minimum=<0 0 0>, Maximum=<1 1 1>))
  ]
  rendererClasses [&NiagaraMeshRendererProperties]
  renderers [{rendererClass=&NiagaraMeshRendererProperties, rendererIndex=0}]
 }
 Ribbon_Forward {
  emitterUpdateScript [
   &EmitterState(Life Cycle Mode=::NewEnumerator1, Loop Duration=0.1, UseLoopDelay=false)
   &FN_SpawnRibbon_v2(SpawnRate:NiagaraFloat=&FN_FrontEndFloatModifier(
     Gameplay=&Lerp_Float(A=0, B=40, Alpha=&ClampFloat(Float=$User.ForwardDot))
     Front End=40
    ))
  ]
  particleSpawnScript [
   /Game/Effects/Niagara/Modules/Ribbons/FN_InitializeRibbon_v4(
    Custom Offset=true
    Enable Skeletal Position=true
    Lifetime:NiagaraFloat=&FN_FrontEndFloatModifier(Gameplay=0.35, Front End=0.5)
    Local Vector Offset:Vector3f=&MakeVector(
     X=20
     Y=&FN_FrontEndFloatModifier(Gameplay=0, Front End=15)
     Z=&FN_FrontEndFloatModifier(Gameplay=5, Front End=5)
    )
    Offset Amount=1
    Ribbon Width=50
    Sampled Position Coordinate Space:ENiagaraCoordinateSpace=::Local
    Sampled Rotation Coordinate Space:ENiagaraCoordinateSpace=::Local
    Sampled Velocity Coordinate Space:ENiagaraCoordinateSpace=::Local
    SkeletalMesh:NiagaraDataInterfaceSkeletalMesh=$User.SkeletalMesh
   )
   &AddVelocity(Velocity=&FN_FrontEndVector3Modifier(Gameplay=&MakeVector(X=0, Y=900, Z=-200), Front End=<0 700 200>)) {module=AddVelocity001}
   &AddVelocity(Velocity=&Multiply_VectorByFloat(Vector=<200 0 0>, Float=$User.RightDot))
   &DynamicMaterialParameters(
    Index 0 Param 1=1
    Index 0 Param 2=0.3
    Index 0 Param 3=0.4
    Index 0 Param 4=0.6
    Module.Param0WriteEnabled=true
    Module.Param1WriteEnabled=true
    Module.Param2WriteEnabled=true
    Module.Param3WriteEnabled=true
   ) {module=DynamicMaterialParameters002}
  ]
  particleUpdateScript [
   &SolveForcesAndVelocity(Clamp Velocity=false, Limit Acceleration=false)
   /Game/Effects/Niagara/Modules/Ribbons/FN_UpdateRibbon_v4(
    Enable Skeletal Position=true
    Sampled Position Coordinate Space:ENiagaraCoordinateSpace=::Local
    Sampled Rotation Coordinate Space:ENiagaraCoordinateSpace=::Local
    Sampled Velocity Coordinate Space:ENiagaraCoordinateSpace=::Local
    SkeletalMesh:NiagaraDataInterfaceSkeletalMesh=$User.SkeletalMesh
   )
   &ScaleColor(
    Scale Alpha=&FN_FrontEndFloatModifier(Gameplay=$System.FN_Integrator.Value, Front End=1)
    ScaleA=true
    ScaleRGB=false
   )
   &ScaleRibbonWidth(Ribbon Width Scale=&Add_Float(
     A=&FloatFromCurve(
      FloatCurve=DataInterface() {propertyValues {Curve {keys [
          {arriveTangent=0.1712, interpMode=RCIM_Cubic, leaveTangent=0.1712, tangentMode=RCTM_User, time=0, value=0.4}
          {interpMode=RCIM_Cubic, tangentMode=RCTM_User, time=0.8, value=1}
         ]}}}
      Scale Curve=2
     )
     B=0
    )) {module=ScaleRibbonWidth002}
   /SetVariables(Particles.Ribbon_width:NiagaraFloat=&Multiply_Float(
     A=-50
     B=&Add_Float(
      A=&FN_FrontEndFloatModifier(
       Gameplay=&FloatFromCurve(
        FloatCurve=DataInterface() {propertyValues {Curve {keys [
            {arriveTangent=0.1712, interpMode=RCIM_Cubic, leaveTangent=0.1712, tangentMode=RCTM_User, time=0, value=0.4}
            {interpMode=RCIM_Cubic, tangentMode=RCTM_User, time=0.8, value=0.85}
           ]}}}
        Scale Curve=2.5
       )
       Front End=&FloatFromCurve(
        FloatCurve=DataInterface() {propertyValues {Curve {keys [
            {arriveTangent=0.1712, interpMode=RCIM_Cubic, leaveTangent=0.1712, tangentMode=RCTM_User, time=0, value=0.4}
            {interpMode=RCIM_Cubic, tangentMode=RCTM_User, time=0.8, value=1.4}
           ]}}}
        Scale Curve=2
       )
      )
      B=0
     )
    )) {isSetParameters=true, module=SetVariables_#1}
   /SetVariables(Particles.RibbonFacing=<1 0 0>) {isSetParameters=true, module=SetVariables_#2}
  ]
  rendererClasses [&NiagaraRibbonRendererProperties]
  renderers [{rendererClass=&NiagaraRibbonRendererProperties}]
 }
 Ribbon_IdleBackward {
  emitterUpdateScript [
   &EmitterState(Life Cycle Mode=::NewEnumerator1, Loop Duration=0.1, UseLoopDelay=false)
   &FN_SpawnRibbon_v2(SpawnRate:NiagaraFloat=&FN_FrontEndFloatModifier(
     Gameplay=&Lerp_Float(A=30, B=30, Alpha=&ClampFloat(Float=$User.ForwardDot))
     Front End=0
    ))
  ]
  particleSpawnScript [
   &FN_InitializeRibbon_v4(
    Custom Offset=true
    Enable Skeletal Position=true
    Lifetime=0.35
    Local Vector Offset:Vector3f=<1 0 0.1>
    Offset Amount=20
    Ribbon Width=50
    Sampled Position Coordinate Space:ENiagaraCoordinateSpace=::Local
    Sampled Rotation Coordinate Space:ENiagaraCoordinateSpace=::Local
    Sampled Velocity Coordinate Space:ENiagaraCoordinateSpace=::Local
    SkeletalMesh:NiagaraDataInterfaceSkeletalMesh=$User.SkeletalMesh
   )
   &AddVelocity(Velocity=<0 300 0>) {module=AddVelocity001}
  ]
  particleUpdateScript [
   &SolveForcesAndVelocity(Clamp Velocity=false, Limit Acceleration=false)
   &FN_UpdateRibbon_v4(
    Enable Skeletal Position=true
    Sampled Position Coordinate Space:ENiagaraCoordinateSpace=::Local
    Sampled Rotation Coordinate Space:ENiagaraCoordinateSpace=::Local
    Sampled Velocity Coordinate Space:ENiagaraCoordinateSpace=::Local
    SkeletalMesh:NiagaraDataInterfaceSkeletalMesh=$User.SkeletalMesh
   )
   &ScaleColor(
    Scale Alpha=&FN_FrontEndFloatModifier(
     Gameplay=&Multiply_Float(
      A=$System.FN_Integrator.Value
      B=&Lerp_Float(A=1, B=1, Alpha=&ClampFloat(Float=$User.ForwardDot))
     )
     Front End=1
    )
    ScaleA=true
    ScaleRGB=false
   ) {enabled=false}
   &DynamicMaterialParameters(
    Index 0 Param 1=1
    Index 0 Param 2=0.3
    Index 0 Param 3=0.4
    Index 0 Param 4=&Lerp_Float(A=0.6, B=2, Alpha=&ClampFloat(Float=$User.ForwardDot))
    Module.Param0WriteEnabled=true
    Module.Param1WriteEnabled=true
    Module.Param2WriteEnabled=true
    Module.Param3WriteEnabled=true
   )
   &ScaleRibbonWidth(Ribbon Width Scale=&Add_Float(
     A=&FloatFromCurve(
      FloatCurve=DataInterface() {propertyValues {Curve {keys [
          {arriveTangent=0.1712, interpMode=RCIM_Cubic, leaveTangent=0.1712, tangentMode=RCTM_User, time=0, value=0.4}
          {interpMode=RCIM_Cubic, tangentMode=RCTM_User, time=0.8, value=1}
         ]}}}
      Scale Curve=2
     )
     B=0
    )) {module=ScaleRibbonWidth002}
   /SetVariables(Particles.Ribbon_width:NiagaraFloat=&Multiply_Float(
     A=-50
     B=&Add_Float(
      A=&FloatFromCurve(
       FloatCurve=DataInterface() {propertyValues {Curve {keys [
           {arriveTangent=0.1712, interpMode=RCIM_Cubic, leaveTangent=0.1712, tangentMode=RCTM_User, time=0, value=0.4}
           {interpMode=RCIM_Cubic, tangentMode=RCTM_User, time=0.8, value=0.85}
          ]}}}
       Scale Curve=2.5
      )
      B=0
     )
    )) {isSetParameters=true, module=SetVariables_#1}
   /SetVariables(Particles.RibbonFacing=<1 0 0>) {isSetParameters=true, module=SetVariables_#2}
  ]
  rendererClasses [&NiagaraRibbonRendererProperties]
  renderers [{rendererClass=&NiagaraRibbonRendererProperties}]
 }
 Splash {
  emitterUpdateScript [
   &EmitterState()
   &SpawnRate(SpawnRate=0, Use Spawn Probability=false)
   /Game/Effects/Niagara/Modules/Ribbons/FN_SpawnRibbon_v3(
    Number of Ribbons=2
    Skeleton Mesh:NiagaraDataInterfaceSkeletalMesh=$User.SkeletalMesh
    Spawn Mode:ENiagaraSkeletalSpawning=::NewEnumerator1
    SpawnRate:NiagaraFloat=&FN_FrontEndFloatModifier(
     Gameplay=&Lerp_Float(A=6, B=20, Alpha=&ClampFloat(Float=$User.ForwardDot))
     Front End=20
    )
   )
  ]
  particleSpawnScript [
   &FN_SetSpawnGroup(Number of Spawn Groups=2)
   /Game/Effects/Niagara/Modules/Ribbons/FN_InitializeRibbon_v5(
    Custom Offset=true
    Enable Skeletal Position=true
    Lifetime=0.35
    Local Vector Offset:Vector3f=&SelectVectorFromArray(Direct Array Index=$Particles.SpawnGroup, Vector Selection Array=DataInterface() {propertyValues {FloatData [<1 -1 0>, <-1 -1 0>]}})
    Offset Amount=3
    Ribbon Width=60
    SkeletalMesh:NiagaraDataInterfaceSkeletalMesh=$User.SkeletalMesh
   )
   &AddVelocity(Velocity=&FN_FrontEndVector3Modifier(
     Gameplay=&Multiply_Vector(
      A=<0 400 10>
      B=&MakeVector(X=1, Y=&Lerp_Float(A=0.5, B=1, Alpha=&ClampFloat(Float=$User.ForwardDot)), Z=1)
     )
     Front End=&RandomRangeVector(Minimum=<0 500 75>, Maximum=<0 700 150>)
    )) {module=AddVelocity001}
   &AddVelocity(Velocity=&Multiply_VectorByFloat(Vector=<200 0 0>, Float=$User.RightDot))
   /SetVariables(Particles.Color=&LerpLinearColors(
     StartColor=&ColorFromCurve(
      ColorCurve=DataInterface() {propertyValues {
        AlphaCurve {keys [<at 0 1 RCIM_Cubic>]}
        BlueCurve {keys [<at 0 1 RCIM_Cubic>,<at 0.0929 1>,<at 0.5021 0>,<at 0.7704 0.2097>,<at 1 0.3701 RCIM_Cubic>]}
        GreenCurve {keys [<at 0 0.7559 RCIM_Cubic>,<at 0.0929 0.0537>,<at 0.5021 0.8217>,<at 0.7704 0>,<at 1 0.0135 RCIM_Cubic>]}
        RedCurve {keys [<at 0 0 RCIM_Cubic>,<at 0.0929 0>,<at 0.5021 1>,<at 0.7704 0.651>,<at 1 1 RCIM_Cubic>]}
       }}
      CurveIndex=&RandomRangeFloat(Minimum=0, Maximum=1)
     )
     EndColor=&ColorFromCurve(
      ColorCurve=DataInterface() {propertyValues {
        AlphaCurve {keys [<at 0 1 RCIM_Cubic>]}
        BlueCurve {keys [<at 0 1 RCIM_Cubic>,<at 0.0929 1>,<at 0.5021 0>,<at 0.7704 0.2097>,<at 1 0.3701 RCIM_Cubic>]}
        GreenCurve {keys [<at 0 0.7559 RCIM_Cubic>,<at 0.0929 0.0537>,<at 0.5021 0.6901>,<at 0.7704 0>,<at 1 0.0135 RCIM_Cubic>]}
        RedCurve {keys [<at 0 0 RCIM_Cubic>,<at 0.0929 0>,<at 0.5021 1>,<at 0.7704 0.651>,<at 1 1 RCIM_Cubic>]}
       }}
      CurveIndex=&RandomRangeFloat(Minimum=0, Maximum=1)
     )
     LerpFactor=&MakeFloatFromInt(Int=$Particles.SpawnGroup)
    )) {isSetParameters=true}
   /SetVariables(Particles.MaterialRandom:NiagaraFloat=&RandomRangeFloat(Minimum=0, Maximum=1)) {isSetParameters=true}
  ]
  particleUpdateScript [
   &ScaleVelocity(Velocity Scale=&MakeVector(
     X=1
     Y=&Multiply_Float(
      A=0.9
      B=&FloatFromCurve(
       FloatCurve=DataInterface() {propertyValues {Curve {keys [<at 0 0.9>,<at 1 3>]}}}
       CurveIndex=&Divide_Float(A=&VectorLength(Vector=$Engine.Owner.Velocity), B=9000)
      )
     )
     Z=1
    )) {module=ScaleVelocity001}
   &SolveForcesAndVelocity(Clamp Velocity=false, Limit Acceleration=false)
   /Game/Effects/Niagara/Modules/Ribbons/FN_UpdateRibbon_v6(Enable Skeletal Position=true, SkeletalMesh:NiagaraDataInterfaceSkeletalMesh=$User.SkeletalMesh)
   &DynamicMaterialParameters(
    Index 0 Param 1=&MaskFloatBySpawnGroup(Masked Float=0.3, Passthrough Float=1)
    Index 0 Param 2=&FloatFromCurve(FloatCurve=DataInterface() {propertyValues {Curve {keys [
         {arriveTangent=-0.3062, interpMode=RCIM_Cubic, leaveTangent=-0.3062, tangentMode=RCTM_User, time=0, value=0.5}
         {arriveTangent=0.0167, interpMode=RCIM_Cubic, leaveTangent=0.0167, tangentMode=RCTM_Break, time=1, value=0}
        ]}}})
    Index 0 Param 3=1
    Index 0 Param 4=1
    Module.Param0WriteEnabled=true
    Module.Param1WriteEnabled=true
    Module.Param2WriteEnabled=true
    Module.Param3WriteEnabled=true
   )
   &ScaleRibbonWidth(Ribbon Width Scale=&FloatFromCurve(
     FloatCurve=DataInterface() {propertyValues {Curve {keys [
         {arriveTangent=-0.0063, interpMode=RCIM_Cubic, leaveTangent=-0.0063, tangentMode=RCTM_User, time=0, value=1}
         {arriveTangent=2.3571, interpMode=RCIM_Cubic, leaveTangent=2.3571, time=0.4, value=2.025}
         <at 0.7 2.325>
        ]}}}
     Scale Curve=0.6
    ))
   &ScaleColor(
    Scale Alpha=&Multiply_Float(
     A=&FNTimeOfDayScalar(Night Scale=0.75, Morning Scale=1, Day Scale=0.5, Evening Scale=1)
     B=&FloatFromCurve(FloatCurve=DataInterface() {propertyValues {Curve {keys [<at 0 0>,<at 0.1 1>,<at 1 1>]}}})
    )
    Scale RGB=&Multiply_Vector(
     A=&MakeVectorFromLinearColorRGB(Color=<rgba 1 1 1 1>)
     B=&VectorFromFloat(Value=2)
    )
    ScaleA=true
    ScaleRGB=true
   ) {module=ScaleColor001}
  ]
  rendererClasses [&NiagaraRibbonRendererProperties]
  renderers [{rendererClass=&NiagaraRibbonRendererProperties}]
 }
}
summary {
 systemName=NS_Glider_Vivid_Male_Trails
 emitters [
  {emitterName=Meshes, rendererClasses [&NiagaraMeshRendererProperties]}
  {emitterName=Confetti, rendererClasses [&NiagaraSpriteRendererProperties]}
  {emitterName=Splash, rendererClasses [&NiagaraRibbonRendererProperties]}
  {emitterName=AttachedGlows, rendererClasses [&NiagaraSpriteRendererProperties]}
  {emitterName=FN_Contrails_ConeStreak, rendererClasses [&NiagaraMeshRendererProperties]}
  {emitterName=Ribbon_IdleBackward, rendererClasses [&NiagaraRibbonRendererProperties]}
  {emitterName=Light, rendererClasses [&NiagaraLightRendererProperties]}
  {emitterName=Ribbon_Forward, rendererClasses [&NiagaraRibbonRendererProperties]}
 ]
 userVariables [
  User.bIsFrontEnd {type=NiagaraBool}
  User.bIsFrontEndPreview {type=NiagaraBool}
  User.bIsFullyDeployed {type=NiagaraBool}
  User.ForwardDot {type=NiagaraFloat, defaultValue=1, description=""}
  User.PlayerSpeed {type=NiagaraFloat, defaultValue=0, description=""}
  User.RibbonIDOffset {type=NiagaraInt32}
  User.RightDot {type=NiagaraFloat, defaultValue=0, description=""}
  User.SkeletalMesh {
   type=NiagaraDataInterfaceSkeletalMesh
   defaultValue {
    struct=NiagaraExt_VariableValue_DataInterface
    value {dataInterface=&self:NiagaraDataInterfaceSkeletalMesh_0}
   }
  }
 ]
}
